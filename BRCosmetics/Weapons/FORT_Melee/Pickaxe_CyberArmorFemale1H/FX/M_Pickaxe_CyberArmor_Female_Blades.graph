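# Unlit, translucent, two-sided material for Niagara mesh particles in the blade effects of the female CyberArmor pickaxe.
# Saturated particle color feeds an LDR-to-HDR derivation as both input and tint, with luminance derived from linear.
# A dynamic parameter, IntensityModifier, scales the HDR intensity per particle.
# The remapped result passes through the shared FX emissive adjustment function.
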Material /BRCosmetics/Weapons/FORT_Melee/Pickaxe_CyberArmorFemale1H/FX/M_Pickaxe_CyberArmor_Female_Blades
graph {
 properties {
  bUsedWithNiagaraMeshParticles=true
  blendMode=BLEND_Translucent
  pixelDepthOffsetMode=PDOM_Legacy
  shadingModel=MSM_Unlit
  twoSided=true
  usedShadingModels=Unlit
 }
 nodes [
  n1 = Saturate(None=ParticleColor().RGB)
  MaterialFunctionCall(Emissive=MaterialFunctionCall(
    LDR Input=@n1
    HDR Tint=@n1
    HDR Intensity Multiplier=DynamicParameter() {defaultValue=<rgba 0.25 1 1 1>, paramNames [IntensityModifier, Param2, Param3, Param4]}.IntensityModifier
    Derive Luminance from Linear=StaticBool() {value=true}
   ) {materialFunction=&DeriveHDRfromLDR}.RemappedRange) {materialFunction=&MF_FX_EmissiveAdjustments}
 ]
}
