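# Placeholder-style prototype material in the FX folder of the CyberArmorFemale1H pickaxe cosmetic, using Default Lit shading.
# It is flagged for use on skeletal meshes and with mesh deformers.
# The graph holds only a dark grey Base Color vector parameter and three scalar constants, with no textures and no wiring shown between nodes or to outputs.
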Material /BRCosmetics/Weapons/FORT_Melee/Pickaxe_CyberArmorFemale1H/FX/M_SKProto_ProtoMaterial
graph {
 properties {bUsedWithMeshDeformer=true, bUsedWithSkeletalMesh=true, pixelDepthOffsetMode=PDOM_Legacy, usedShadingModels=Default Lit}
 nodes [
  Constant() {r=1}
  Constant() {r=0.333}
  Constant() {r=0.5}
  VectorParameter() {parameterName=Base Color, defaultValue=<rgba 0.0868 0.0868 0.0868 0>}
 ]
}
