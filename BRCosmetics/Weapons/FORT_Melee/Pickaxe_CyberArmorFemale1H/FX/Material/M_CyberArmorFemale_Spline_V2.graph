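# Translucent, two-sided unlit material for spline FX on the CyberArmorFemale one-handed pickaxe, usable on Niagara mesh particles and ribbons.
# SplineThicken takes base and tip widths from dynamic parameters plus clamped camera-distance scaling, with three panned cylinder-gradient samples added as offset noise.
# Shading nodes combine particle color, emissive adjustment, UV gradients, a ramp and panning cloud noise.
# Several nodes appear unused, as do the PanSpeed and PowerGrow parameters.
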Material /BRCosmetics/Weapons/FORT_Melee/Pickaxe_CyberArmorFemale1H/FX/Material/M_CyberArmorFemale_Spline_V2
graph {
 properties {
  bUsedWithNiagaraMeshParticles=true
  bUsedWithNiagaraRibbons=true
  blendMode=BLEND_Translucent
  pixelDepthOffsetMode=PDOM_Legacy
  shadingModel=MSM_Unlit
  twoSided=true
  usedShadingModels=Unlit
 }
 nodes [
  n1 = DynamicParameter() {defaultValue=<rgba 1 0 1 1>, paramNames [Spline Base, Spline Tip, PanSpeed, PowerGrow]}
  n2 = Multiply(A=@n1.Spline Base, B=ScalarParameter() {defaultValue=2, group=WPO, parameterName=WidthBase})
  n3 = ObjectPositionWS()
  n4 = CameraPositionWS()
  n5 = Clamp(
   None=Divide(A=Distance(A=@n3, B=@n4), B=ScalarParameter() {defaultValue=1400, group=Scale By Distance, parameterName=Distance Scale})
   Min=ScalarParameter() {group=Scale By Distance, parameterName=DistanceScaleMinClamp}
   Max=ScalarParameter() {defaultValue=50, group=Scale By Distance, parameterName=DistanceScaleMaxClamp}
  ) {maxDefault=10}
  n6 = Multiply(A=@n1.Spline Tip, B=ScalarParameter() {defaultValue=1, group=WPO, parameterName=WidthTop})
  n7 = ScalarParameter() {defaultValue=1, group=WPO, parameterName=LengthNoiseTile}
  n8 = Multiply(A=TextureCoordinate(), B=@n7)
  n9 = Multiply(A=TextureCoordinate() {vTiling=2}, B=@n7)
  n10 = Add(
   A=TextureSample(UVs=Panner(Coordinate=@n8) {speedY=1}) {texture=/Game/Global/Textures/Default/Masks_and_Gradients/CylGradient}.G
   B=TextureSample(UVs=Panner(Coordinate=@n9) {speedY=1}) {texture=&CylGradient}.G
  )
  n11 = TextureCoordinate() {vTiling=6}
  n12 = TextureSample(UVs=Panner(Coordinate=Multiply(A=@n11, B=@n7)) {speedY=1}) {texture=&CylGradient}
  Add(
   A=MaterialFunctionCall(WidthBase=Add(A=@n5, B=@n2), WidthTip=Add(A=@n6, B=@n5)) {materialFunction=&SplineThicken}.WorldPositionOffset
   B=Multiply(A=Add(A=@n10, B=@n12.G), B=ScalarParameter() {defaultValue=2.5, group=WPO, parameterName=LengthNoise})
  )
  Constant()
  Constant()
  Constant() {r=1}
  Constant() {r=0.1}
  n13 = OneMinus(None=ComponentMask(None=TextureCoordinate()) {r=true})
  n14 = TextureSample(UVs=Panner(Coordinate=TextureCoordinate() {uTiling=2, vTiling=2}, Speed=Constant2Vector() {r=1}) {speedY=-0.3}) {texture=&AtmosphericCloudNoise01}
  Divide(A=@n13, B=@n14.RGB)
  n15 = TextureCoordinate()
  MaterialFunctionCall!(UVs=@n15, CenterPosition, Radius) {materialFunction=/Engine/Functions/Engine_MaterialFunctions01/Gradient/SphereGradient-2D}
  n16 = ParticleColor()
  MaterialFunctionCall(Emissive=@n16.RGB) {materialFunction=&MF_FX_EmissiveAdjustments}
  Multiply(A=Multiply(A=@n13, B=@n14.RGB), B=@n16.RGB)
  n17 = ComponentMask(None=TextureCoordinate()) {g=true}
  n18 = Multiply(A=Multiply(A=OneMinus(None=@n17), B=@n17)) {constB=2}
  n19 = TextureCoordinate() {vTiling=1.4}
  n20 = Subtract(A=@n14.RGB, B=OneMinus(None=ComponentMask(None=@n19) {g=true}))
  Multiply(
   A=Power(Base=Saturate(None=@n18)) {constExponent=1}
   B=Multiply(A=TextureSample() {texture=/Game/Global/Textures/Default/Masks_and_Gradients/T_Ramp_Virt}.RGB, B=OneMinus(None=@n20))
  )
  Reroute()
  SphereMask(A=@n15, B=Constant2Vector() {g=1})
  TextureSample() {texture=/Game/Gadgets/Assets/Misty_Bop/Effects/T_MediumGradient_Horizontal}
 ]
}
