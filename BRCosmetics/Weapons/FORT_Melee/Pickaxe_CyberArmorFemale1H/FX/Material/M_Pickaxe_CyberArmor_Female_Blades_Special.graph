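# Translucent, unlit Niagara mesh-particle material for special blade effects on the female Cyber Armor pickaxe.
# Panning flame and circuit-board textures blend with a world-space sphere mask, TV static and optional rainbow tint, swapped through dynamic parameters.
# The particle-colored result becomes HDR, then gains vertex-color inner and outer glows before emissive adjustment.
# Particle alpha, clamped to a high minimum, drives opacity, with depth fade except on mobile.
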Material /BRCosmetics/Weapons/FORT_Melee/Pickaxe_CyberArmorFemale1H/FX/Material/M_Pickaxe_CyberArmor_Female_Blades_Special
graph {
 properties {
  bUsedWithNiagaraMeshParticles=true
  blendMode=BLEND_Translucent
  pixelDepthOffsetMode=PDOM_Legacy
  shadingModel=MSM_Unlit
  twoSided=true
  usedShadingModels=Unlit
 }
 nodes [
  n1 = ParticleColor()
  n2 = TextureCoordinate()
  n3 = Constant2Vector() {g=0.5}
  n4 = Power(Base=TextureSample(UVs=Panner(Coordinate=@n2, Speed=@n3) {speedY=-0.3}) {samplerType=SAMPLERTYPE_LinearGrayscale, texture=/BRCosmetics/Weapons/FORT_Melee/Pickaxe_Cube_Ninja_Male/FX/T_FluidFlameTile}.R) {constExponent=0.5}
  n5 = TextureSample(UVs=TextureCoordinate() {uTiling=6, vTiling=6}) {texture=/Game/Effects/Fort_Effects/Textures/T_FX_Circuitboard}
  n6 = Panner(Coordinate=TextureCoordinate() {uTiling=3, vTiling=2}) {speedY=1}
  n7 = Constant() {r=-0.25}
  n8 = Add(
   A=TextureSampleParameter2D!(
    UVs=MaterialFunctionCall(UVs=@n6, "Rotation Angle (0-1)"=@n7) {materialFunction=&CustomRotator}.Rotated Values
    Level
    Apply View MipBias
   ) {mipValueMode=TMVM_MipLevel, parameterName=Noise Tex, texture=/Game/Effects/Fort_Effects/Textures/Tile/Tests/T_Tile_Panner_Mask}.G
   B=ScalarParameter() {defaultValue=0.1, group=Tech, parameterName=Tech Noise Add}
  )
  n9 = Multiply(A=Multiply(A=Power(Base=@n8)) {constB=2}, B=VectorParameter() {parameterName=Dark_Cool, defaultValue=<rgba 0.5196 0.8993 1 0>}.RGB)
  n10 = Power(Base=Add(A=@n5.G, B=Multiply(A=@n9, B=@n5.G))) {constExponent=8}
  n11 = DynamicParameter() {parameterIndex=1, defaultValue=<rgba 0.25 1 1 1>, paramNames [MatSwapB,MatSwapC,Param3,Param4]}
  n12 = TextureSample(UVs=Panner(Coordinate=TextureCoordinate(), Speed=Constant2Vector() {g=0.5}) {speedY=-0.3}) {samplerType=SAMPLERTYPE_LinearGrayscale, texture=&T_FluidFlameTile}
  n13 = TextureSample(UVs=TextureCoordinate() {uTiling=6, vTiling=8}) {texture=&T_FX_Circuitboard}
  n14 = DynamicParameter() {parameterIndex=2, defaultValue=<rgba 0 0 0 65>, paramNames [X,Y,Z,Radius]}
  n15 = SphereMask(A=WorldPosition().XYZ, B=AppendVector(A=AppendVector(A=@n14.X, B=@n14.X), B=@n14.Z), Radius=@n14.Radius, Hardness=Constant() {r=-2})
  n16 = TextureCoordinate() {uTiling=2, vTiling=2}
  n17 = Multiply(A=TextureSample(UVs=Panner(Coordinate=@n16) {speedY=-0.25}) {samplerType=SAMPLERTYPE_LinearGrayscale, texture=&T_FluidFlameTile}.RGB) {constB=-0.15}
  n18 = Multiply(A=@n15) {constB=0.7}
  n19 = Clamp(None=Power(Base=LinearInterpolate(A=@n17, Alpha=@n18) {constB=2})) {minDefault=0.01}
  n20 = DynamicParameter() {parameterIndex=1, defaultValue=<rgba 0.25 1 1 1>, paramNames [MatSwapB,MatSwapC,Param3,Param4]}
  n21 = Panner(Coordinate=TextureCoordinate() {uTiling=5, vTiling=5}) {speedX=0.25}
  n22 = DynamicParameter() {defaultValue=<rgba 0.25 1 1 1>, paramNames [IntensityModifier, VertexColorOutside, VertexColorInside, MatSwap]}
  n23 = LinearInterpolate(
   A=Constant() {r=1}
   B=Power(Base=TextureSample(UVs=@n21) {texture=/Game/Global/Textures/Default/Masks_and_Gradients/FortniteRainbow}.RGB) {constExponent=-0.3}
   Alpha=@n22.MatSwap
  )
  n24 = TextureCoordinate()
  n25 = Constant2Vector() {g=0.5}
  n26 = Add(
   A=TextureSample(UVs=Panner(Coordinate=@n24, Speed=@n25) {speedY=-0.3}) {texture=/Game/Athena/Items/Gameplay/Second_Chance/Textures/T_Van_TV_Static_01_D}.RGB
   B=Constant() {r=0.9}
  )
  n27 = Add(
   A=LinearInterpolate(A=Multiply(A=@n12.RGB, B=@n13.G), B=Multiply(A=@n4, B=@n10), Alpha=@n11.MatSwapC)
   B=Multiply(A=LinearInterpolate(A=@n15, B=@n19, Alpha=@n20.MatSwapB), B=Multiply(A=@n23, B=@n26))
  )
  n28 = DynamicParameter() {defaultValue=<rgba 0.25 1 1 1>, paramNames [IntensityModifier, VertexColorOutside, VertexColorInside, MatSwap]}
  n29 = MaterialFunctionCall(
   LDR Input=Clamp(None=Multiply(A=@n1.RGB, B=@n27))
   HDR Tint=Saturate(None=@n1.RGB)
   HDR Intensity Multiplier=@n28.IntensityModifier
   Derive Luminance from Linear=StaticBool() {value=true}
  ) {materialFunction=&DeriveHDRfromLDR}
  n30 = DynamicParameter() {defaultValue=<rgba 0.25 1 1 1>, paramNames [IntensityModifier, VertexColorOutside, VertexColorInside, MatSwap]}
  n31 = Divide(A=VertexColor().B, B=VertexColor().G)
  n32 = Multiply(A=Clamp(None=Multiply(A=@n30.VertexColorInside, B=@n31)) {maxDefault=30}, B=ParticleColor().RGB)
  n33 = DynamicParameter() {defaultValue=<rgba 0.25 1 1 1>, paramNames [IntensityModifier, VertexColorOutside, VertexColorInside, MatSwap]}
  n34 = Multiply(
   A=@n33.VertexColorOutside
   B=MaterialFunctionCall!(X=VertexColor().G, Tangent 0=Constant(), Tangent 1=Constant() {r=2.2}) {materialFunction=/Engine/Functions/Engine_MaterialFunctions02/Gradients/SmoothCurve}.Result
  )
  n35 = ParticleColor()
  MaterialFunctionCall(Emissive=Add(A=Add(A=@n29.RemappedRange, B=@n32), B=Multiply(A=@n34, B=@n35.RGB))) {materialFunction=&MF_FX_EmissiveAdjustments}
  MaterialFunctionCall(CenterPosition=Constant2Vector() {g=0.5, r=0.5}, Radius=Constant() {r=0.5}, Density=Constant() {r=0.8}) {materialFunction=&RadialGradientExponential}
  n36 = Clamp(None=ParticleColor().A) {minDefault=0.7}
  n37 = DepthFade(Opacity=@n36, FadeDistance=ScalarParameter() {defaultValue=3, parameterName=DepthDistance})
  ShadingPathSwitch(Default=@n37, Deferred=@n37, Mobile=@n36) {desc=Migrated from FeatureLevelSwitch}
 ]
}
