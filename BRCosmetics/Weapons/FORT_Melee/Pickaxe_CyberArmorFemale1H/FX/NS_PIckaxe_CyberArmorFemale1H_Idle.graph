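# Niagara idle effect for the CyberArmorFemale1H pickaxe, drawing its glowing, user-configurable blades.
# Blade_A/B/C emitters hold persistent mesh particles, with user floats choosing each mesh index and driving material swaps.
# A and C index lookup arrays by mesh to place and size glow sprites; A adds a light, and Flares adds two offset sprites.
# All emitters tint from Top Blade Color; blades add system sine-driven alpha and time-of-day color scaling.
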NiagaraSystem /BRCosmetics/Weapons/FORT_Melee/Pickaxe_CyberArmorFemale1H/FX/NS_PIckaxe_CyberArmorFemale1H_Idle
dependencies {
 usedDataInterfaces [&NiagaraDataInterfaceArrayFloat3, &NiagaraDataInterfaceArrayFloat2]
 usedDynamicInputs [&RandomRangeFloat, &Lerp_Float, &Waveform]
 usedModules [
  &self:SystemSpawnScript.SystemScriptSource.SystemScriptGraph.NiagaraNodeAssignment_2.SetVariables_#1
  &self:SystemSpawnScript.SystemScriptSource.SystemScriptGraph.NiagaraNodeAssignment_3.SetVariables_#2
  &SystemState
  &self:SystemSpawnScript.SystemScriptSource.SystemScriptGraph.NiagaraNodeAssignment_0.SetVariables_#3
  &self:SystemSpawnScript.SystemScriptSource.SystemScriptGraph.NiagaraNodeAssignment_6.SetVariables_#4
 ]
}
emitters {
 Blade_A {
  emitterUpdateScript [
   &EmitterState(Life Cycle Mode=::NewEnumerator1, Loop Behavior=::NewEnumerator1, Loop Duration Mode=::NewEnumerator1, UseLoopDelay=false)
   &SpawnBurst_Instantaneous(Use Spawn Probability=false, UseLoopCountLimit=false)
  ]
  particleSpawnScript [&InitializeParticle(Lifetime=0, Mesh Scale Mode=::NewEnumerator3, Sprite Size Mode=::NewEnumerator3, UsePositionOffset=false, Write Material Random=true)]
  particleUpdateScript [
   &ParticleState(Kill Particles When Lifetime Has Elapsed=false, Let Infinitely Lived Particles Die When Emitter Deactivates=true)
   /SetVariables(Particles.MeshIndex=&MakeIntFromFloat(Float=$User.BladeASelect)) {isSetParameters=true, module=SetVariables_#5}
   /SetVariables(Particles.SpritePoint:Vector3f=&SelectVectorFromArray(
     Direct Array Index=$Particles.MeshIndex
     Vector Selection Array=DataInterface() {propertyValues {FloatData [
        <0 0 1>
        <1.4 0 1>
        <0 0 1>
        <0.7 0 1>
        <0 0 1>
        <0.6 0 1>
        <-0.3 0 1>
        <1.7 0 1>
        <0 0 1>
        <1 0 1>
        <0 0 1>
        <0 0 1>
        <0 0 1>
        <1 0 1>
        <-0.5 0 1>
        <2 0 1>
       ]}}
    )) {isSetParameters=true}
   &Color(
    Color=&MakeLinearColorFromVectorAndFloat(
     "Vector (RGB)"=&MakeVectorFromLinearColorRGB(Color=$User.Top Blade Color)
     "Float (Alpha)"=$System.SinMultiply
    )
    Scale Alpha=0.75
    Scale Color=&VectorFromFloat(Value=&FNTimeOfDayScalar(Night Scale=1, Morning Scale=1, Day Scale=1.5, Evening Scale=1))
   )
   &DynamicMaterialParameters(
    Index 0 Param 1=&Multiply_Float(
     A=&MakeFloatFromLinearColor(LinearColor=$User.Top Blade Color, Channel:ENiagaraLinearColor_Channels=::NewEnumerator3)
     B=$User.BladeBrightness
    )
    Index 0 Param 3=0.8
    Index 0 Param 4=$User.MatSwap
    Index 1 Param 1=$User.MatSwapB
    Index 1 Param 2=$User.MatSwapC
    Module.Param0WriteEnabled=true
    Module.Param0WriteEnabled1=true
    Module.Param1WriteEnabled=true
    Module.Param1WriteEnabled1=true
    Module.Param2WriteEnabled=true
    Module.Param2WriteEnabled1=false
    Module.Param3WriteEnabled=true
    Module.Param3WriteEnabled1=false
    Write Parameter Index 1=true
   )
   /SetVariables(Particles.SphereLocation:Vector4f=&MakeFloat4FromVectorAndFloat(
     "Vector (RGB)"=&ConvertPositionToVector(Input Position=$Engine.Owner.Position, Convert To Absolute Large World Vector=false)
     "Float (Alpha)"=65
    )) {isSetParameters=true}
   &FN_Light_Attributes(Create Exponent Attribute=true, Create Position Attribute=true, Create Radius Attribute=true, Exponent=16, Position=<0 0 25>, Radius=40)
   /SetVariables(
    Particles.GlowDynParams:Vector4f=<3 1 1 1>
    Particles.GlowPos:NiagaraPosition=&AddVectorToPosition(
     Position=$Engine.Emitter.SimulationPosition
     Vector=&Add_Vector(
      A=$Particles.SpritePoint
      B=&SelectVectorFromArray(
       Direct Array Index=$Particles.MeshIndex
       Vector Selection Array=DataInterface() {propertyValues {FloatData [
          <0 0 40>
          <0 0 27>
          <-7 0 40>
          <-10 0 40>
          <0 0 55>
          <-15 0 41>
          <-20 0 30>
          <-13 0 68>
          <0 0 23>
          <2 0 25>
          <-3 0 27>
          <-3 0 30>
          <-1.5 0 40>
          <-12 0 30>
          <-3 0 27>
          <-10 0 35>
         ]}}
      )
     )
    )
    Particles.SpriteSize=&SelectVector2DFromArray(
     Direct Array Index=$Particles.MeshIndex
     Vector 2D Selection Array=DataInterface() {propertyValues {FloatData [
        <35 80>
        <35 95>
        <28 100>
        <45 80>
        <25 130>
        <35 80>
        <45 80>
        <22 60>
        <25 50>
        <30 70>
        <28 80>
        <30 55>
        <30 70>
        <20 63>
        <27 50>
        <35 70>
       ]}}
    )
   ) {isSetParameters=true}
   /SetVariables(
    Particles.EdgingColor:LinearColor=&MultiplyLinearColor(
     Linear Color=$Particles.Color
     Scale Factor=&MakeVector4FromVectorAndFloat(
      "Vector (RGB)"=<200 200 200>
      "Float (Alpha)"=&Multiply_Float(
       A=$System.SinMultiply
       B=&FNTimeOfDayScalar(Night Scale=0.5, Morning Scale=0.5, Evening Scale=0.8)
      )
     )
    )
    Particles.EdgingDynParam:Vector4f=<1 0 1 1>
   ) {isSetParameters=true}
  ]
  rendererClasses [
   &NiagaraMeshRendererProperties
   &NiagaraLightRendererProperties
   &NiagaraSpriteRendererProperties
  ]
  renderers [
   {rendererClass=&NiagaraMeshRendererProperties, rendererIndex=0}
   {rendererClass=&NiagaraLightRendererProperties, rendererIndex=1}
   {rendererClass=&NiagaraSpriteRendererProperties, rendererIndex=2}
   {rendererClass=&NiagaraMeshRendererProperties, rendererIndex=3}
  ]
 }
 Blade_B {
  emitterUpdateScript [
   &EmitterState(Life Cycle Mode=::NewEnumerator1, Loop Behavior=::NewEnumerator1, Loop Duration Mode=::NewEnumerator1, UseLoopDelay=false)
   &SpawnBurst_Instantaneous(Use Spawn Probability=false, UseLoopCountLimit=false)
  ]
  particleSpawnScript [&InitializeParticle(Lifetime=0, Mesh Scale Mode=::NewEnumerator3, UsePositionOffset=false, Write Material Random=true)]
  particleUpdateScript [
   &ParticleState(Kill Particles When Lifetime Has Elapsed=false, Let Infinitely Lived Particles Die When Emitter Deactivates=true)
   /SetVariables(Particles.MeshIndex=&MakeIntFromFloat(Float=$User.BladeBSelect)) {isSetParameters=true, module=SetVariables_#5}
   &Color(
    Color=&MakeLinearColorFromVectorAndFloat(
     "Vector (RGB)"=&MakeVectorFromLinearColorRGB(Color=$User.Top Blade Color)
     "Float (Alpha)"=$System.SinMultiply3
    )
    Scale Alpha=0.75
    Scale Color=&VectorFromFloat(Value=&FNTimeOfDayScalar(Night Scale=1, Morning Scale=1, Day Scale=1.5, Evening Scale=1))
   )
   &DynamicMaterialParameters(
    Index 0 Param 1=&Multiply_Float(
     A=&MakeFloatFromLinearColor(LinearColor=$User.Top Blade Color, Channel:ENiagaraLinearColor_Channels=::NewEnumerator3)
     B=$User.BladeBrightness
    )
    Index 0 Param 3=0.8
    Index 0 Param 4=$User.MatSwap
    Index 1 Param 1=$User.MatSwapB
    Index 1 Param 2=$User.MatSwapC
    Module.Param0WriteEnabled=true
    Module.Param0WriteEnabled1=true
    Module.Param1WriteEnabled=true
    Module.Param1WriteEnabled1=true
    Module.Param2WriteEnabled=true
    Module.Param2WriteEnabled1=false
    Module.Param3WriteEnabled=true
    Module.Param3WriteEnabled1=false
    Write Parameter Index 1=true
   ) {module=DynamicMaterialParameters001}
   /SetVariables(Particles.SphereLocation:Vector4f=&MakeFloat4FromVectorAndFloat(
     "Vector (RGB)"=&ConvertPositionToVector(Input Position=$Engine.Owner.Position, Convert To Absolute Large World Vector=false)
     "Float (Alpha)"=65
    )) {isSetParameters=true}
  ]
  rendererClasses [&NiagaraMeshRendererProperties]
  renderers [{rendererClass=&NiagaraMeshRendererProperties, rendererIndex=0}]
 }
 Blade_C {
  emitterUpdateScript [
   &EmitterState(Life Cycle Mode=::NewEnumerator1, Loop Behavior=::NewEnumerator1, Loop Duration Mode=::NewEnumerator1, UseLoopDelay=false)
   &SpawnBurst_Instantaneous(Use Spawn Probability=false, UseLoopCountLimit=false)
  ]
  particleSpawnScript [&InitializeParticle(Lifetime=0, Mesh Scale Mode=::NewEnumerator3, UsePositionOffset=false, Write Material Random=true)]
  particleUpdateScript [
   &ParticleState(Kill Particles When Lifetime Has Elapsed=false, Let Infinitely Lived Particles Die When Emitter Deactivates=true)
   /SetVariables(Particles.MeshIndex=&MakeIntFromFloat(Float=$User.BladeCSelect)) {isSetParameters=true, module=SetVariables_#5}
   /SetVariables(Particles.SpritePoint:Vector3f=&SelectVectorFromArray(
     Direct Array Index=$Particles.MeshIndex
     Vector Selection Array=DataInterface() {propertyValues {FloatData [
        <0 0 -1>
        <1 0 -1>
        <0 0 -1>
        <0 0 -1>
        <0 0 -1>
        <1 0 -1>
        <-0.5 0 -1>
        <2.4 0 -1>
        <0 0 -1>
        <1.4 0 -1>
        <0 0 -1>
        <0.7 0 -1>
        <0 0 -1>
        <0.6 0 -1>
        <-0.3 0 -1>
        <1.7 0 -1>
       ]}}
    )) {isSetParameters=true}
   &Color(
    Color=&MakeLinearColorFromVectorAndFloat(
     "Vector (RGB)"=&MakeVectorFromLinearColorRGB(Color=$User.Top Blade Color)
     "Float (Alpha)"=$System.SinMultiply
    )
    Scale Alpha=0.75
    Scale Color=&VectorFromFloat(Value=&FNTimeOfDayScalar(Night Scale=1, Morning Scale=1, Day Scale=1.5, Evening Scale=1))
   )
   &DynamicMaterialParameters(
    Index 0 Param 1=&Multiply_Float(
     A=&MakeFloatFromLinearColor(LinearColor=$User.Top Blade Color, Channel:ENiagaraLinearColor_Channels=::NewEnumerator3)
     B=$User.BladeBrightness
    )
    Index 0 Param 3=0.8
    Index 0 Param 4=$User.MatSwap
    Index 1 Param 1=$User.MatSwapB
    Index 1 Param 2=$User.MatSwapC
    Module.Param0WriteEnabled=true
    Module.Param0WriteEnabled1=true
    Module.Param1WriteEnabled=true
    Module.Param1WriteEnabled1=true
    Module.Param2WriteEnabled=true
    Module.Param2WriteEnabled1=false
    Module.Param3WriteEnabled=true
    Module.Param3WriteEnabled1=false
    Write Parameter Index 1=true
   ) {module=DynamicMaterialParameters001}
   /SetVariables(Particles.SphereLocation:Vector4f=&MakeFloat4FromVectorAndFloat(
     "Vector (RGB)"=&ConvertPositionToVector(Input Position=$Engine.Owner.Position, Convert To Absolute Large World Vector=false)
     "Float (Alpha)"=65
    )) {isSetParameters=true}
   /SetVariables(
    Particles.GlowDynParam:Vector4f=<3 1 1 1>
    Particles.GlowPos:NiagaraPosition=&AddVectorToPosition(
     Position=$Engine.Emitter.SimulationPosition
     Vector=&Add_Vector(
      A=$Particles.SpritePoint
      B=&SelectVectorFromArray(
       Direct Array Index=$Particles.MeshIndex
       Vector Selection Array=DataInterface() {propertyValues {FloatData [
          <2 0 25>
          <-30 0 10>
          <-3 0 30>
          <-2 0 20>
          <0 0 30>
          <-40 0 2>
          <18 0 13>
          <-40 0 -25>
          <-1 0 -20>
          <-20 0 -10>
          <-7 0 10>
          <-25 0 -20>
          <0 0 0>
          <-34 0 -15>
          <-15 0 -10>
          <-35 0 -53>
         ]}}
      )
     )
    )
    Particles.SpriteSize=&SelectVector2DFromArray(
     Direct Array Index=$Particles.MeshIndex
     Vector 2D Selection Array=DataInterface() {propertyValues {FloatData [
        <35 80>
        <35 95>
        <28 100>
        <45 80>
        <25 130>
        <35 80>
        <45 80>
        <22 60>
        <25 50>
        <30 70>
        <28 80>
        <30 55>
        <30 70>
        <20 63>
        <27 50>
        <35 70>
       ]}}
    )
   ) {isSetParameters=true}
  ]
  rendererClasses [&NiagaraMeshRendererProperties, &NiagaraSpriteRendererProperties]
  renderers [
   {rendererClass=&NiagaraMeshRendererProperties, rendererIndex=0}
   {rendererClass=&NiagaraSpriteRendererProperties, rendererIndex=1}
  ]
 }
 Flares {
  emitterUpdateScript [
   &EmitterState(Life Cycle Mode=::NewEnumerator1, Loop Behavior=::NewEnumerator1, Loop Duration Mode=::NewEnumerator1, UseLoopDelay=false)
   &SpawnBurst_Instantaneous(Use Spawn Probability=false, UseLoopCountLimit=false)
   &SpawnBurst_Instantaneous(Spawn Group=1, Use Spawn Probability=false, UseLoopCountLimit=false) {module=SpawnBurst_Instantaneous001}
  ]
  particleSpawnScript [
   &InitializeParticle(
    Lifetime=0
    Mesh Scale Mode=::NewEnumerator3
    Position Offset=&SelectVectorFromArray(Direct Array Index=$Emitter.SpawnGroup, Vector Selection Array=DataInterface() {propertyValues {FloatData [<-4 0 9>, <-4 0 -10.25>]}})
    Sprite Rotation Mode=::NewEnumerator2
    Sprite Size Mode=::NewEnumerator3
    Uniform Sprite Size=25
    UsePositionOffset=true
    Write Material Random=true
   )
   &DynamicMaterialParameters(
    Index 0 Param 1=0.5
    Index 0 Param 2=0.75
    Index 0 Param 3=1
    Index 1 Param 1=5
    Module.Param0WriteEnabled=true
    Module.Param0WriteEnabled1=true
    Module.Param1WriteEnabled=true
    Module.Param1WriteEnabled1=true
    Module.Param2WriteEnabled=true
    Module.Param2WriteEnabled1=true
    Module.Param3WriteEnabled=true
    Module.Param3WriteEnabled1=true
    Write Parameter Index 1=true
   )
   &CameraOffset(Camera Offset Amount=3)
  ]
  particleUpdateScript [
   &ParticleState(Kill Particles When Lifetime Has Elapsed=false, Let Infinitely Lived Particles Die When Emitter Deactivates=true)
   &Color(
    Color=&MakeLinearColorFromVectorAndFloat("Vector (RGB)"=&MakeVectorFromLinearColorRGB(Color=$User.Top Blade Color), "Float (Alpha)"=1)
    Scale Color=&VectorFromFloat(Value=10)
   )
  ]
  rendererClasses [&NiagaraSpriteRendererProperties]
  renderers [{rendererClass=&NiagaraSpriteRendererProperties}]
 }
}
summary {
 systemName=NS_PIckaxe_CyberArmorFemale1H_Idle
 emitters [
  {
   emitterName=Blade_A
   rendererClasses [
    &NiagaraMeshRendererProperties
    &NiagaraLightRendererProperties
    &NiagaraSpriteRendererProperties
   ]
  }
  {emitterName=Blade_B, rendererClasses [&NiagaraMeshRendererProperties]}
  {emitterName=Blade_C, rendererClasses [&NiagaraMeshRendererProperties, &NiagaraSpriteRendererProperties]}
  {emitterName=Flares, rendererClasses [&NiagaraSpriteRendererProperties]}
 ]
 userVariables [
  User.BladeASelect {type=NiagaraFloat, defaultValue=0, description=""}
  User.BladeBrightness {type=NiagaraFloat, defaultValue=25, description=""}
  User.BladeBSelect {type=NiagaraFloat, defaultValue=0, description=""}
  User.BladeCSelect {type=NiagaraFloat, defaultValue=0, description=""}
  User.Bottom Blade Color {type=LinearColor, description="", defaultValue {struct=LinearColor, value=<rgba 1 1 1 1>}}
  User.Forward Blade Color {type=LinearColor, description="", defaultValue {struct=LinearColor, value=<rgba 1 1 1 1>}}
  User.GlowStrength {type=NiagaraFloat, defaultValue=1, description=""}
  User.MatSwap {type=NiagaraFloat, defaultValue=0, description=""}
  User.MatSwapB {type=NiagaraFloat, defaultValue=0, description=""}
  User.MatSwapC {type=NiagaraFloat, defaultValue=0, description=""}
  User.Top Blade Color {type=LinearColor, description="", defaultValue {struct=LinearColor, value=<rgba 1 1 1 1>}}
 ]
}
